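; A Dark Nexus surveillance camera prop that tracks the local player cosmetically through its mesh material while they are inside its overlap area.
; Construction writes the Arrow component's world location to custom primitive data and to the "Player Pos" material parameter as the initial target.
; When the local player's pawn begins overlap, a looping timer feeds that pawn's location into "Player Pos".
; End overlap clears the timer without resetting the parameter.
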(event Custom|Start%20Track
  (Networking|CanExecuteCosmeticEvents
    (:True
      (bind _returnvalue (Utilities|Time|SetTimerbyEvent (delegate Custom|Timer%20Tick) 0.1 true true)))))

(event Custom|Timer%20Tick
  (bind _output_get (Variables|Default|SetPlayer%20Location (Transformation|GetActorLocation (Variables|Default|GetPlayer%20Pawn))))
  (Rendering|Material|SetVectorParameterValueonMaterials (Variables|Default|GetSM_DarkNexus_SurveillanceCamera_A) "Player Pos" _output_get))

(event Collision|EventActorBeginOverlap
  (params (OtherActor :type Actor%20Object%20Reference))
  ; Overlap
  (Networking|CanExecuteCosmeticEvents
    (:True
      (if (== OtherActor (Pawn|GetControlledPawn (Fortnite|GetFirstLocalFortPlayerController true)))
        (Variables|Default|SetPlayer%20Pawn OtherActor)
        (CallFunction|StartTrack)))))

(event Collision|EventActorEndOverlap
  (params (OtherActor :type Actor%20Object%20Reference))
  ; Overlap
  (Networking|CanExecuteCosmeticEvents
    (:True
      (if (== (Variables|Default|GetPlayer%20Pawn) OtherActor)
        (Utilities|Time|ClearandInvalidateTimerbyHandle (stale Custom|Start%20Track _returnvalue))))))

(fn ConstructionScript
  (bind _output_get (Variables|Default|SetRelative%20Location (Transformation|GetWorldLocation (Variables|Default|GetArrow))))
  (Rendering|Material|SetDefaultCustomPrimitiveDataVector3 (Variables|Default|GetSM_DarkNexus_SurveillanceCamera_A) 0 _output_get)
  (Rendering|Material|SetVectorParameterValueonMaterials (Variables|Default|GetSM_DarkNexus_SurveillanceCamera_A) "Player Pos" _output_get)
  (|CallParentFunction))
