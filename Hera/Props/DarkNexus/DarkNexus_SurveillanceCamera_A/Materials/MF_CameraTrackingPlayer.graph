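# MF_CameraTrackingPlayer is a material function stored with the DarkNexus Surveillance Camera A prop, named for making the camera track the player.
# Its graph holds only a function output node, with no inputs, expressions or connections shown.
# No tracking logic appears in the graph, so the function reads as an empty stub.
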MaterialFunction /Hera/Props/DarkNexus/DarkNexus_SurveillanceCamera_A/Materials/MF_CameraTrackingPlayer
graph {nodes [FunctionOutput()]}
