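# Material function that yaws a mesh around its pivot to face a tracked position, defaulting to the camera, for the DarkNexus surveillance camera prop.
# The XY direction from object to target feeds a fast arctangent, remapped to a 0–1 turn.
# RotateAboutAxis applies it around world Z, outputting WPO alongside the angle.
# An unconnected bool input's name notes that object rotation is ignored.
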MaterialFunction /Hera/Props/DarkNexus/DarkNexus_SurveillanceCamera_A/Materials/MF_RotationTrackCharacter
graph {nodes [
  n1 = FunctionInput(Preview=CameraPositionWS()) {bUsePreviewValueAsDefault=true, inputName=PositionToTrack}
  n2 = ObjectPositionWS()
  n3 = Normalize(VectorInput=ComponentMask(None=Subtract(A=@n1, B=@n2)) {g=true, r=true})
  n4 = ComponentMask(None=@n3) {g=true}
  n5 = ComponentMask(None=@n3) {r=true}
  Arctangent2!(Y=@n4, X=@n5)
  FunctionInput() {bUsePreviewValueAsDefault=true, inputName="(DoesNotAcountForObjectRotation)*", inputType=FunctionInput_Bool, sortPriority=3}
  n6 = Divide(
   A=Add(A=Arctangent2Fast(Y=@n4, X=@n5), B=MaterialFunctionCall() {materialFunction=&Pi}.Result)
   B=MaterialFunctionCall(Multiplier=Constant() {r=2}) {materialFunction=&Pi}.Result
  )
  n7 = WorldPosition() {worldPositionShaderOffset=WPT_ExcludeAllShaderOffsets}
  FunctionOutput(None=RotateAboutAxis(
    NormalizedRotationAxis=Constant3Vector() {constant=<rgba 0 0 1 0>}
    RotationAngle=@n6
    PivotPoint=ObjectPositionWS()
    Position=TruncateLWC!(None=@n7.XYZ)
   )) {outputName=Rotation WPO}
  FunctionOutput(None=@n6) {outputName=Rotation Angle}
 ]}
