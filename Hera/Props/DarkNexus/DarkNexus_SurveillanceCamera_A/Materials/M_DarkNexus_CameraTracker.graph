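# Surveillance camera material for the Dark Nexus prop that turns the mesh in the vertex shader to track a player position.
# MF_RotationTrackCharacter aims between a custom-primitive-data forward position and Player Pos, adding rotation and sine bob to world position offset.
# Normals are corrected for rotation; emissive adds pulsing bloom with eye-adaptation compensation, alongside pixel-depth fade masks.
# The attributes pass through global emissive multiplier and destructible damage functions with additive WPO.
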Material /Hera/Props/DarkNexus/DarkNexus_SurveillanceCamera_A/Materials/M_DarkNexus_CameraTracker
graph {
 properties {bAutomaticallySetUsageInEditor=false, bUseMaterialAttributes=true, usedShadingModels=Default Lit}
 nodes [
  n1 = VectorParameter() {bUseCustomPrimitiveData=true, parameterName=Foward Pos, defaultValue=<rgba 0 0 0 1>}
  Add(A=@n1.RGB) {constB=0.01}
  CameraPositionWS()
  ComponentMask(None=Saturate()) {g=true}
  Constant3Vector() {constant=<rgba -10992.6914 0 0 0>}
  Constant3Vector()
  MaterialFunctionCall() {materialFunction=&ObjectPivotPoint}
  n2 = MaterialFunctionCall!(
   PositionToTrack=LinearInterpolate(A=@n1.RGB, B=VectorParameter() {parameterName=Player Pos, defaultValue=<rgba 0 0 0 1>}.RGB, Alpha=NamedRerouteUsage())
   "(DoesNotAcountForObjectRotation)*"
  ) {materialFunction=&/MF_RotationTrackCharacter}
  n3 = MaterialFunctionCall(
   Rotation Axis=Constant3Vector() {constant=<rgba 0 0 1 0>}
   Rotation Angle=NamedRerouteDeclaration(None=@n2.Rotation Angle) {name=Rotation Angle, nodeColor=<rgba 1 0.3281 0 1>}
  ) {materialFunction=&FixRotateAboutAxisNormals}
  n4 = MaterialFunctionCall!(Sine Phase=Divide(A=Time()) {constB=4}, Value 1=Constant() {r=-2}, Value 2=Constant() {r=2}) {materialFunction=/Engine/Functions/Engine_MaterialFunctions02/Utility/Sine_Remapped}
  n5 = NamedRerouteDeclaration(None=Multiply(A=Constant3Vector() {constant=<rgba 0 0 1 0>}, B=@n4.Remapped Sine)) {name=Bob, nodeColor=<rgba 0.4297 1 0 1>}
  n6 = SetMaterialAttributes!(
   MaterialAttributes
   Base Color=NamedRerouteUsage()
   Metallic=NamedRerouteUsage()
   Specular=NamedRerouteUsage()
   Roughness=NamedRerouteUsage()
   Emissive Color=NamedRerouteUsage()
   Normal=MaterialFunctionCall(BaseNormal=VertexInterpolator(VS=@n3.New Tangent Space Vertex Normal).PS, AdditionalNormal=NamedRerouteUsage()) {materialFunction=&BlendAngleCorrectedNormals}.Result
   World Position Offset=NamedRerouteDeclaration(None=Add(A=@n2.Rotation WPO, B=@n5)) {name=Offset, nodeColor=<rgba 0.9922 1 0 1>}
  ) {attributeSetTypes [
    #1
    #2
    #3
    #4
    #5
    #6
    #7
   ]}
  MaterialFunctionCall(
   Material_In=MaterialFunctionCall!(In=@n6) {materialFunction=/Game/Global/MaterialFunctions/MF_GlobalEmissiveMultiplier_MPC}.Result
   UseAdditiveWPO=StaticBool() {value=true}
  ) {materialFunction=&MF_FN_DestructibleDamage}
  MotionVectorWorldOffsetOutput!(None)
  n7 = TextureSampleParameter2D() {
   parameterName=SFX
   samplerType=SAMPLERTYPE_Masks
   texture=/Hera/Props/DarkNexus/DarkNexus_SurveillanceCamera_A/Textures/T_DarkNexus_SurveillanceCamera_A_A_M
  }
  NamedRerouteDeclaration(None=Multiply(A=@n7.G, B=Constant3Vector() {constant=<rgba 5 1.3765 0.0485 1>})) {name=Emissive Map, nodeColor=<rgba 1 0.8203 0 1>}
  n8 = TextureSampleParameter2D() {parameterName=Base Color, texture=/Hera/Props/DarkNexus/DarkNexus_SurveillanceCamera_A/Textures/T_DarkNexus_SurveillanceCamera_A_A_D}
  NamedRerouteDeclaration(None=@n8.RGB) {name=Base Color Map, nodeColor=<rgba 1 0.8203 0 1>}
  n9 = PixelDepth()
  n10 = Divide(A=Add(A=@n9.R, B=Constant() {r=-1200}), B=Constant() {r=200})
  NamedRerouteDeclaration(None=OneMinus(None=Saturate(None=@n10))) {name=Distance 2, nodeColor=<rgba 1 0.7734 0 1>}
  NamedRerouteDeclaration(None=@n7.R) {name=Emissive Map 2, nodeColor=<rgba 1 0.8203 0 1>}
  n11 = TextureSampleParameter2D() {
   parameterName=Normal
   samplerType=SAMPLERTYPE_Normal
   texture=/Hera/Props/DarkNexus/DarkNexus_SurveillanceCamera_A/Textures/T_DarkNexus_SurveillanceCamera_A_A_N
  }
  NamedRerouteDeclaration(None=@n11.RGB) {name=Normal Map, nodeColor=<rgba 1 0.8203 0 1>}
  n12 = TextureSampleParameter2D() {
   parameterName=RMS
   samplerType=SAMPLERTYPE_Masks
   texture=/Hera/Props/DarkNexus/DarkNexus_SurveillanceCamera_A/Textures/T_DarkNexus_SurveillanceCamera_A_A_S
  }
  NamedRerouteDeclaration(None=@n12.G) {name=Specular, nodeColor=<rgba 1 0.8203 0 1>}
  NamedRerouteDeclaration(None=@n12.R) {name=Metal, nodeColor=<rgba 1 0.8203 0 1>}
  NamedRerouteDeclaration(None=@n12.B) {name=Rough, nodeColor=<rgba 1 0.8203 0 1>}
  n13 = Divide(
   A=Add(A=@n9.R, B=ScalarParameter() {defaultValue=-1500, parameterName=Add Pos})
   B=ScalarParameter() {defaultValue=500, parameterName=Pixel Depth}
  )
  NamedRerouteDeclaration(None=OneMinus(None=Saturate(None=@n13))) {name=Distance, nodeColor=<rgba 1 0.7734 0 1>}
  NamedRerouteDeclaration(None=@n4.Remapped Sine) {name=Sine, nodeColor=<rgba 0 1 1 1>}
  n14 = NamedRerouteUsage()
  n15 = SmoothStep(Min=Constant() {r=0.0944}, Max=Constant() {r=1.9707}, Value=NamedRerouteUsage())
  n16 = Add(A=@n15, B=Multiply(A=@n15, B=@n15))
  n17 = Time()
  n18 = Constant() {r=0.5}
  n19 = Saturate(None=MaterialFunctionCall(Sine Phase=Multiply(A=@n17, B=@n18), Value 1=@n18, Value 2=Constant() {r=1}) {materialFunction=&Sine_Remapped}.Remapped Sine)
  n20 = Add(
   A=@n14
   B=Multiply(A=VectorParameter() {parameterName=Camera bloom, defaultValue=<rgba 0.5885 0.1598 0 1>}.RGB, B=Multiply(A=@n16, B=@n19)) {constB=2}
  )
  n21 = NamedRerouteUsage()
  NamedRerouteDeclaration(None=EyeAdaptationInverse(LightValueInput=LinearInterpolate(A=@n14, B=@n20, Alpha=@n21), AlphaInput=Constant() {r=0.4}).EyeAdaptationInverse) {name=Emissive, nodeColor=<rgba 0 1 0.7891 1>}
  NamedRerouteUsage()
  NamedRerouteUsage()
  ObjectPositionWS()
  Saturate()
  Saturate()
  Subtract(B=WorldPosition().XYZ)
  TransformPosition() {transformType=TRANSFORMPOSSOURCE_World}
  Transform(None=Add(
    A=LocalPosition() {includedOffsets=ExcludeOffsets, localOrigin=InstancePreSkinning}.XYZ
    B=VectorParameter() {defaultValue=<rgba -77608 0 0 1>}.RGB
   )) {transformSourceType=TRANSFORMSOURCE_Local}
  VertexInterpolator()
 ]
}
