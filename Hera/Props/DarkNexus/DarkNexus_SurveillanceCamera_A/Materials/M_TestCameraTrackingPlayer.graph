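# Test material for the DarkNexus surveillance camera prop that turns the mesh toward the viewer's camera through World Position Offset.
# The default "Alternative Tracking" branch yaws vertices around the pivot by a signed angle between flattened camera direction and local axes, with corrected normals.
# The switched-off branch rebuilds vertices and normals from an orthogonalized axis basis.
# Experimental look-at nodes, including player-position and top-angle blending, sit apart from both branches.
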Material /Hera/Props/DarkNexus/DarkNexus_SurveillanceCamera_A/Materials/M_TestCameraTrackingPlayer
graph {
 expressions {
  MaterialExpressionAbs_0 {class=MaterialExpressionAbs, inputs [None {from=MaterialExpressionSign_1, output=""}], properties {}}
  MaterialExpressionAdd_0 {
   class=MaterialExpressionAdd
   inputs [
    A {from=MaterialExpressionMultiply_3}
    B {from=""}
   ]
   properties {constB=0}
  }
  MaterialExpressionAdd_2 {
   class=MaterialExpressionAdd
   inputs [
    A {from=MaterialExpressionGetMaterialAttributes_1, output=World Position Offset}
    B {from=MaterialExpressionGetMaterialAttributes_2, output=World Position Offset}
   ]
  }
  MaterialExpressionAdd_25 {
   class=MaterialExpressionAdd
   inputs [
    A {from=MaterialExpressionSubtract_22}
    B {from=MaterialExpressionMaterialFunctionCall_21, output=Result}
   ]
  }
  MaterialExpressionAdd_3 {
   class=MaterialExpressionAdd
   inputs [
    A {from=MaterialExpressionScalarParameter_67}
    B {from=MaterialExpressionScalarParameter_68}
   ]
   properties {constB=0.005}
  }
  MaterialExpressionAdd_4 {
   class=MaterialExpressionAdd
   inputs [
    A {from=MaterialExpressionCameraPositionWS_0}
    B {from=MaterialExpressionVectorParameter_4, output=RGB}
   ]
  }
  MaterialExpressionAdd_8 {
   class=MaterialExpressionAdd
   inputs [
    A {from=MaterialExpressionComponentMask_4}
    B {from=MaterialExpressionWorldPosition_3, output=XY}
   ]
  }
  MaterialExpressionAppendVector_0 {
   class=MaterialExpressionAppendVector
   inputs [
    A {from=MaterialExpressionComponentMask_0, output=""}
    B {from=MaterialExpressionConstant_1, output=""}
   ]
  }
  MaterialExpressionAppendVector_1 {
   class=MaterialExpressionAppendVector
   inputs [
    A {from=MaterialExpressionComponentMask_1, output=""}
    B {from=MaterialExpressionConstant_1, output=""}
   ]
  }
  MaterialExpressionAppendVector_2 {
   class=MaterialExpressionAppendVector
   inputs [
    A {from=MaterialExpressionAdd_8, output=""}
    B {from=MaterialExpressionConstant_4, output=""}
   ]
  }
  MaterialExpressionAppendVector_5 {
   class=MaterialExpressionAppendVector
   inputs [
    A {from=MaterialExpressionAbs_0, output=""}
    B {from=MaterialExpressionVectorParameter_2, output=A}
   ]
  }
  MaterialExpressionAppendVector_7 {
   class=MaterialExpressionAppendVector
   inputs [
    A {from=MaterialExpressionComponentMask_8, output=""}
    B {from=MaterialExpressionClamp_1, output=""}
   ]
  }
  MaterialExpressionArccosineFast_1 {class=MaterialExpressionArccosineFast, inputs [None {from=MaterialExpressionDotProduct_0, output=""}], properties {}}
  MaterialExpressionCameraPositionWS_0 {class=MaterialExpressionCameraPositionWS, inputs [], properties {}}
  MaterialExpressionCameraPositionWS_1 {class=MaterialExpressionCameraPositionWS, inputs [], properties {}}
  MaterialExpressionCameraPositionWS_3 {class=MaterialExpressionCameraPositionWS, inputs [], properties {}}
  MaterialExpressionCameraVectorWS_0 {class=MaterialExpressionCameraVectorWS, inputs [], properties {}}
  MaterialExpressionClamp_1 {
   class=MaterialExpressionClamp
   inputs [
    None {from=MaterialExpressionComponentMask_7, output=""}
    Min {from=MaterialExpressionScalarParameter_17, output=""}
    Max {from=MaterialExpressionScalarParameter_54, output=""}
   ]
   properties {}
  }
  MaterialExpressionComponentMask_0 {class=MaterialExpressionComponentMask, inputs [None {from=MaterialExpressionNormalize_0}], properties {g=true, r=true}}
  MaterialExpressionComponentMask_1 {class=MaterialExpressionComponentMask, inputs [None {from=MaterialExpressionTransform_0}], properties {g=true, r=true}}
  MaterialExpressionComponentMask_3 {class=MaterialExpressionComponentMask, inputs [None {from=MaterialExpressionTransform_4}], properties {g=true}}
  MaterialExpressionComponentMask_4 {class=MaterialExpressionComponentMask, inputs [None {from=MaterialExpressionVectorParameter_8, output=RGB}], properties {g=true, r=true}}
  MaterialExpressionComponentMask_7 {class=MaterialExpressionComponentMask, inputs [None {from=MaterialExpressionReroute_11}], properties {b=true}}
  MaterialExpressionComponentMask_8 {class=MaterialExpressionComponentMask, inputs [None {from=MaterialExpressionReroute_11}], properties {g=true, r=true}}
  MaterialExpressionConstant3Vector_0 {class=MaterialExpressionConstant3Vector, properties {constant=<rgba 0 0 1 0>}}
  MaterialExpressionConstant3Vector_1 {class=MaterialExpressionConstant3Vector, properties {constant=<rgba 1 0 0 0>}}
  MaterialExpressionConstant3Vector_13 {class=MaterialExpressionConstant3Vector, properties {constant=<rgba 0.1 0.1 1 0>}}
  MaterialExpressionConstant3Vector_14 {class=MaterialExpressionConstant3Vector, properties {constant=<rgba 0 1 0 0>}}
  MaterialExpressionConstant3Vector_15 {class=MaterialExpressionConstant3Vector, properties {constant=<rgba -6580.8457 -139.6751 2700 0>}}
  MaterialExpressionConstant3Vector_16 {class=MaterialExpressionConstant3Vector}
  MaterialExpressionConstant3Vector_18 {class=MaterialExpressionConstant3Vector, properties {constant=<rgba -1 0 0 0>}}
  MaterialExpressionConstant3Vector_19 {class=MaterialExpressionConstant3Vector, properties {constant=<rgba 0 0 1 0>}}
  MaterialExpressionConstant3Vector_2 {class=MaterialExpressionConstant3Vector, properties {constant=<rgba 0 0 -1 0>}}
  MaterialExpressionConstant3Vector_20 {class=MaterialExpressionConstant3Vector, properties {constant=<rgba 0 0 1 0>}}
  MaterialExpressionConstant3Vector_3 {class=MaterialExpressionConstant3Vector, properties {constant=<rgba 0 1 0 0>}}
  MaterialExpressionConstant3Vector_4 {class=MaterialExpressionConstant3Vector, properties {constant=<rgba 0 1 0 0>}}
  MaterialExpressionConstant3Vector_5 {class=MaterialExpressionConstant3Vector, properties {constant=<rgba 1 0 0 0>}}
  MaterialExpressionConstant3Vector_7 {class=MaterialExpressionConstant3Vector, properties {constant=<rgba 0 0 1 0>}}
  MaterialExpressionConstant_1 {class=MaterialExpressionConstant}
  MaterialExpressionConstant_4 {class=MaterialExpressionConstant, properties {r=1}}
  MaterialExpressionConvert_2 {
   class=MaterialExpressionConvert
   inputs [None {from=MaterialExpressionSubtract_2}]
   properties {
    nodeName=Break Float3
    convertInputs [{
      type=Vector3
      expressionInput {
       expression=&self:MaterialExpressionSubtract_2
       inputName=None
       mask=0
       maskA=0
       maskB=0
       maskG=0
       maskR=0
       outputIndex=0
      }
     }]
    convertOutputs [{type=Scalar}, {type=Scalar}, {type=Scalar}]
   }
  }
  MaterialExpressionConvert_3 {
   class=MaterialExpressionConvert
   inputs [
    None {from=""}
    None {from=""}
    None {from=MaterialExpressionConvert_2}
   ]
   properties {
    nodeName=Make Vector3
    convertInputs [
     {type=Scalar, expressionInput {expression=null, inputName=None, mask=0, maskA=0, maskB=0, maskG=0, maskR=0, outputIndex=0}}
     {type=Scalar, expressionInput {expression=null, inputName=None, mask=0, maskA=0, maskB=0, maskG=0, maskR=0, outputIndex=1}}
     {
      type=Scalar
      expressionInput {
       expression=&self:MaterialExpressionConvert_2
       inputName=None
       mask=0
       maskA=0
       maskB=0
       maskG=0
       maskR=0
       outputIndex=2
      }
     }
    ]
    convertOutputs [{type=Vector3}]
   }
  }
  MaterialExpressionConvert_6 {
   class=MaterialExpressionConvert
   inputs [None {from=MaterialExpressionObjectPositionWS_2}]
   properties {
    nodeName=Break Float3
    convertInputs [{
      type=Vector3
      expressionInput {
       expression=&self:MaterialExpressionObjectPositionWS_2
       inputName=None
       mask=0
       maskA=0
       maskB=0
       maskG=0
       maskR=0
       outputIndex=0
      }
     }]
    convertOutputs [{type=Scalar}, {type=Scalar}, {type=Scalar}]
   }
  }
  MaterialExpressionConvert_7 {
   class=MaterialExpressionConvert
   inputs [
    None {from=MaterialExpressionConvert_6}
    None {from=MaterialExpressionConvert_6}
    None {from=MaterialExpressionConvert_6}
   ]
   properties {
    nodeName=Make Vector3
    convertInputs [
     {
      type=Scalar
      expressionInput {
       expression=&self:MaterialExpressionConvert_6
       inputName=None
       mask=0
       maskA=0
       maskB=0
       maskG=0
       maskR=0
       outputIndex=0
      }
     }
     {
      type=Scalar
      expressionInput {
       expression=&self:MaterialExpressionConvert_6
       inputName=None
       mask=0
       maskA=0
       maskB=0
       maskG=0
       maskR=0
       outputIndex=1
      }
     }
     {
      type=Scalar
      expressionInput {
       expression=&self:MaterialExpressionConvert_6
       inputName=None
       mask=0
       maskA=0
       maskB=0
       maskG=0
       maskR=0
       outputIndex=2
      }
     }
    ]
    convertOutputs [{type=Vector3}]
   }
  }
  MaterialExpressionConvert_8 {
   class=MaterialExpressionConvert
   inputs [None {from=MaterialExpressionTransform_6}]
   properties {
    nodeName=Break Float3
    convertInputs [{
      type=Vector3
      expressionInput {
       expression=&self:MaterialExpressionTransform_6
       inputName=None
       mask=0
       maskA=0
       maskB=0
       maskG=0
       maskR=0
       outputIndex=0
      }
     }]
    convertOutputs [{type=Scalar}, {type=Scalar}, {type=Scalar}]
   }
  }
  MaterialExpressionCrossProduct_0 {
   class=MaterialExpressionCrossProduct
   inputs [
    A {from="", output=""}
    B {from="", output=""}
   ]
   properties {}
  }
  MaterialExpressionCrossProduct_1 {
   class=MaterialExpressionCrossProduct
   inputs [
    A {from=MaterialExpressionReroute_10, output=""}
    B {from=MaterialExpressionNamedRerouteDeclaration_7, output=""}
   ]
   properties {}
  }
  MaterialExpressionCrossProduct_3 {
   class=MaterialExpressionCrossProduct
   inputs [
    A {from=MaterialExpressionConstant3Vector_7, output=""}
    B {from=MaterialExpressionNamedRerouteDeclaration_5, output=""}
   ]
   properties {}
  }
  MaterialExpressionDivide_0 {
   class=MaterialExpressionDivide
   inputs [
    A {from=MaterialExpressionTime_0}
    B {from=""}
   ]
   properties {constB=5}
  }
  MaterialExpressionDotProduct_0 {
   class=MaterialExpressionDotProduct
   inputs [
    A {from=MaterialExpressionAppendVector_0, output=""}
    B {from=MaterialExpressionAppendVector_1, output=""}
   ]
   properties {}
  }
  MaterialExpressionDotProduct_1 {
   class=MaterialExpressionDotProduct
   inputs [
    A {from=MaterialExpressionTransform_1, output=""}
    B {from=MaterialExpressionAppendVector_0, output=""}
   ]
   properties {}
  }
  MaterialExpressionDotProduct_3 {
   class=MaterialExpressionDotProduct
   inputs [
    A {from=MaterialExpressionNormalize_10, output=""}
    B {from=MaterialExpressionConstant3Vector_19, output=""}
   ]
   properties {}
  }
  MaterialExpressionGetMaterialAttributes_0 {class=MaterialExpressionGetMaterialAttributes, inputs [None {from=MaterialExpressionSetMaterialAttributes_0, output=""}]}
  MaterialExpressionGetMaterialAttributes_1 {
   class=MaterialExpressionGetMaterialAttributes
   inputs [None {from=MaterialExpressionMaterialFunctionCall_12, output=Result}]
   properties {attributeGetTypes [#1, #2]}
  }
  MaterialExpressionGetMaterialAttributes_2 {
   class=MaterialExpressionGetMaterialAttributes
   inputs [None {from=MaterialExpressionSetMaterialAttributes_0, output=""}]
   properties {attributeGetTypes [#1, #2]}
  }
  MaterialExpressionLinearInterpolate_19 {
   class=MaterialExpressionLinearInterpolate
   inputs [
    A {from=MaterialExpressionNormalize_10}
    B {from=MaterialExpressionReroute_12}
    Alpha {from=MaterialExpressionNamedRerouteDeclaration_45}
   ]
  }
  MaterialExpressionLinearInterpolate_20 {
   class=MaterialExpressionLinearInterpolate
   inputs [
    A {from=MaterialExpressionNormalize_11}
    B {from=MaterialExpressionConstant3Vector_18}
    Alpha {from=MaterialExpressionNamedRerouteUsage_70}
   ]
  }
  MaterialExpressionLinearInterpolate_8 {
   class=MaterialExpressionLinearInterpolate
   inputs [
    A {from=MaterialExpressionCameraPositionWS_3}
    B {from=MaterialExpressionReroute_13}
    Alpha {from=MaterialExpressionMax_1}
   ]
  }
  MaterialExpressionLocalPosition_0 {class=MaterialExpressionLocalPosition, inputs [], properties {}}
  MaterialExpressionMaterialFunctionCall_0 {
   class=MaterialExpressionMaterialFunctionCall
   inputs [
    Rotation Axis {from=MaterialExpressionConstant3Vector_0}
    Rotation Angle {from=MaterialExpressionAdd_0}
    World Space Vertex Normals {from=""}
    Movement Mask {from=""}
    Normals Pivot {from=""}
   ]
   properties {materialFunction=&FixRotateAboutAxisNormals}
  }
  MaterialExpressionMaterialFunctionCall_11 {
   class=MaterialExpressionMaterialFunctionCall
   inputs [Float4 {from=MaterialExpressionAppendVector_5}]
   properties {materialFunction=/Game/Global/MaterialFunctions/MF_BreakComponents}
  }
  MaterialExpressionMaterialFunctionCall_12 {
   class=MaterialExpressionMaterialFunctionCall
   inputs [In {from=MaterialExpressionGetMaterialAttributes_0, output=MaterialAttributes}]
   properties {materialFunction=/Game/Athena/WinterArt/Materials/MF_WPOAnim}
  }
  MaterialExpressionMaterialFunctionCall_13 {
   class=MaterialExpressionMaterialFunctionCall
   inputs [
    R {from=MaterialExpressionNamedRerouteUsage_4}
    G {from=MaterialExpressionNamedRerouteUsage_5}
    B {from=MaterialExpressionNamedRerouteUsage_6}
    A {from=""}
   ]
   properties {materialFunction=&AppendMany}
  }
  MaterialExpressionMaterialFunctionCall_14 {
   class=MaterialExpressionMaterialFunctionCall
   inputs [
    BaseNormal {from=MaterialExpressionGetMaterialAttributes_1, output=Normal}
    AdditionalNormal {from=MaterialExpressionGetMaterialAttributes_2, output=Normal}
   ]
   properties {materialFunction=&BlendAngleCorrectedNormals}
  }
  MaterialExpressionMaterialFunctionCall_2 {
   class=MaterialExpressionMaterialFunctionCall
   inputs [
    Normalized V1 {from=MaterialExpressionAppendVector_0}
    Normalized V2 {from=MaterialExpressionAppendVector_1}
   ]
   properties {materialFunction=/Engine/Functions/Engine_MaterialFunctions02/Math/AngleBetweenVectors}
  }
  MaterialExpressionMaterialFunctionCall_21 {
   class=MaterialExpressionMaterialFunctionCall
   inputs []
   properties {materialFunction=&ComponentPivotLocation}
  }
  MaterialExpressionMaterialFunctionCall_23 {
   class=MaterialExpressionMaterialFunctionCall
   inputs []
   properties {materialFunction=/Engine/Functions/Engine_MaterialFunctions02/WorldPositionOffset/ObjectScale}
  }
  MaterialExpressionMaterialFunctionCall_24 {
   class=MaterialExpressionMaterialFunctionCall
   inputs [
    Vector1 {from=MaterialExpressionTransform_15}
    Vector2 {from=MaterialExpressionTransform_18}
    Vector3 {from=MaterialExpressionTransform_19}
   ]
   properties {materialFunction=/Engine/Functions/Engine_MaterialFunctions02/Math/MakeVectorsOrthogonal}
  }
  MaterialExpressionMaterialFunctionCall_26 {
   class=MaterialExpressionMaterialFunctionCall
   inputs [
    VectorToTransform {from=MaterialExpressionMultiply_8}
    BasisX {from=MaterialExpressionMaterialFunctionCall_24, output=Vector1}
    BasisY {from=MaterialExpressionMaterialFunctionCall_24, output=Vector1}
    BasisZ {from=MaterialExpressionMaterialFunctionCall_24, output=Vector1}
    BasisPosition {from=""}
   ]
   properties {materialFunction=&Transform3x3Matrix}
  }
  MaterialExpressionMaterialFunctionCall_3 {
   class=MaterialExpressionMaterialFunctionCall
   inputs []
   properties {materialFunction=&ComponentPivotLocation}
  }
  MaterialExpressionMaterialFunctionCall_4 {
   class=MaterialExpressionMaterialFunctionCall
   inputs [
    VectorToRotate {from=MaterialExpressionTransform_5}
    LookAtVector {from=MaterialExpressionTransform_3}
    RestingVector {from=""}
   ]
   properties {materialFunction=/Engine/Functions/Engine_MaterialFunctions01/Vectors/RotateVector}
  }
  MaterialExpressionMaterialFunctionCall_44 {
   class=MaterialExpressionMaterialFunctionCall
   inputs [
    VectorToTransform {from=MaterialExpressionVertexNormalWS_5}
    BasisX {from=MaterialExpressionMaterialFunctionCall_24, output=Vector1}
    BasisY {from=MaterialExpressionMaterialFunctionCall_24, output=Vector1}
    BasisZ {from=MaterialExpressionMaterialFunctionCall_24, output=Vector1}
    BasisPosition {from=""}
   ]
   properties {materialFunction=&Transform3x3Matrix}
  }
  MaterialExpressionMaterialFunctionCall_7 {
   class=MaterialExpressionMaterialFunctionCall
   inputs []
   properties {materialFunction=&ComponentPivotLocation}
  }
  MaterialExpressionMax_0 {
   class=MaterialExpressionMax
   inputs [
    A {from=MaterialExpressionMaterialFunctionCall_11, output=R}
    B {from=MaterialExpressionMaterialFunctionCall_11, output=R}
   ]
  }
  MaterialExpressionMax_1 {
   class=MaterialExpressionMax
   inputs [
    A {from=MaterialExpressionMax_0}
    B {from=MaterialExpressionMaterialFunctionCall_11, output=B}
   ]
  }
  MaterialExpressionMultiply_1 {
   class=MaterialExpressionMultiply
   inputs [
    A {from=MaterialExpressionReroute_0}
    B {from=MaterialExpressionMultiply_9}
   ]
   properties {constA=1}
  }
  MaterialExpressionMultiply_12 {
   class=MaterialExpressionMultiply
   inputs [
    A {from=MaterialExpressionNormalize_5}
    B {from=""}
   ]
   properties {constB=-1}
  }
  MaterialExpressionMultiply_14 {
   class=MaterialExpressionMultiply
   inputs [
    A {from=MaterialExpressionConstant3Vector_2}
    B {from=MaterialExpressionSign_2}
   ]
  }
  MaterialExpressionMultiply_15 {
   class=MaterialExpressionMultiply
   inputs [
    A {from=MaterialExpressionNormalize_6}
    B {from=MaterialExpressionConstant3Vector_13}
   ]
  }
  MaterialExpressionMultiply_16 {
   class=MaterialExpressionMultiply
   inputs [
    A {from=MaterialExpressionStep_9}
    B {from=""}
   ]
   properties {constB=0.9}
  }
  MaterialExpressionMultiply_3 {
   class=MaterialExpressionMultiply
   inputs [
    A {from=MaterialExpressionMultiply_1}
    B {from=MaterialExpressionScalarParameter_0}
   ]
  }
  MaterialExpressionMultiply_6 {
   class=MaterialExpressionMultiply
   inputs [
    A {from=MaterialExpressionArccosineFast_1}
    B {from=""}
   ]
   properties {constB=57.2958, desc=Output in degrees}
  }
  MaterialExpressionMultiply_8 {
   class=MaterialExpressionMultiply
   inputs [
    A {from=MaterialExpressionMaterialFunctionCall_23, output=Scale XYZ}
    B {from=MaterialExpressionTransformPosition_6}
   ]
   properties {desc=Apply Object Scale once again}
  }
  MaterialExpressionMultiply_9 {
   class=MaterialExpressionMultiply
   inputs [
    A {from=MaterialExpressionMultiply_6}
    B {from=""}
   ]
   properties {constB=0.0056}
  }
  MaterialExpressionNamedRerouteDeclaration_3 {
   class=MaterialExpressionNamedRerouteDeclaration
   inputs [None {from=MaterialExpressionLinearInterpolate_8}]
   properties {name="LookAtPosition(Face)", nodeColor=<rgba 0 0.625 1 1>}
  }
  MaterialExpressionNamedRerouteDeclaration_45 {
   class=MaterialExpressionNamedRerouteDeclaration
   inputs [None {from=MaterialExpressionMultiply_16}]
   properties {name=LookAtAngleHeightMask, nodeColor=<rgba 0 1 0.0156 1>}
  }
  MaterialExpressionNamedRerouteDeclaration_46 {
   class=MaterialExpressionNamedRerouteDeclaration
   inputs [None {from=MaterialExpressionSphereMask_9}]
   properties {name=LookAtAngleHeightMask_Smooth, nodeColor=<rgba 0 1 0.0156 1>}
  }
  MaterialExpressionNamedRerouteDeclaration_5 {
   class=MaterialExpressionNamedRerouteDeclaration
   inputs [None {from=MaterialExpressionAppendVector_7}]
   properties {name=X, nodeColor=<rgba 0 1 0.4609 1>}
  }
  MaterialExpressionNamedRerouteDeclaration_7 {
   class=MaterialExpressionNamedRerouteDeclaration
   inputs [None {from=MaterialExpressionStaticSwitchParameter_3}]
   properties {name=Y, nodeColor=<rgba 0 1 0.4609 1>}
  }
  MaterialExpressionNamedRerouteDeclaration_8 {
   class=MaterialExpressionNamedRerouteDeclaration
   inputs [None {from=MaterialExpressionNormalize_12}]
   properties {name=Z, nodeColor=<rgba 0 1 0.4609 1>}
  }
  MaterialExpressionNamedRerouteUsage_3 {class=MaterialExpressionNamedRerouteUsage}
  MaterialExpressionNamedRerouteUsage_4 {class=MaterialExpressionNamedRerouteUsage}
  MaterialExpressionNamedRerouteUsage_5 {class=MaterialExpressionNamedRerouteUsage}
  MaterialExpressionNamedRerouteUsage_6 {class=MaterialExpressionNamedRerouteUsage}
  MaterialExpressionNamedRerouteUsage_70 {class=MaterialExpressionNamedRerouteUsage}
  MaterialExpressionNormalize_0 {class=MaterialExpressionNormalize, inputs [VectorInput {from=MaterialExpressionSubtract_0, output=""}]}
  MaterialExpressionNormalize_1 {class=MaterialExpressionNormalize, inputs [VectorInput {from=MaterialExpressionConstant3Vector_4, output=""}]}
  MaterialExpressionNormalize_10 {class=MaterialExpressionNormalize, inputs [VectorInput {from=MaterialExpressionSubtract_4, output=""}]}
  MaterialExpressionNormalize_11 {class=MaterialExpressionNormalize, inputs [VectorInput {from=MaterialExpressionCrossProduct_3, output=""}]}
  MaterialExpressionNormalize_12 {class=MaterialExpressionNormalize, inputs [VectorInput {from=MaterialExpressionCrossProduct_1, output=""}]}
  MaterialExpressionNormalize_13 {class=MaterialExpressionNormalize, inputs [VectorInput {from="", output=""}]}
  MaterialExpressionNormalize_2 {class=MaterialExpressionNormalize, inputs [VectorInput {from=MaterialExpressionCameraVectorWS_0, output=""}]}
  MaterialExpressionNormalize_3 {class=MaterialExpressionNormalize, inputs [VectorInput {from=MaterialExpressionMaterialFunctionCall_4, output=Result}]}
  MaterialExpressionNormalize_4 {class=MaterialExpressionNormalize, inputs [VectorInput {from=MaterialExpressionLocalPosition_0, output=XYZ}]}
  MaterialExpressionNormalize_5 {class=MaterialExpressionNormalize, inputs [VectorInput {from=MaterialExpressionWorldPosition_0, output=XYZ}]}
  MaterialExpressionNormalize_6 {class=MaterialExpressionNormalize, inputs [VectorInput {from=MaterialExpressionObjectPositionWS_4, output=""}]}
  MaterialExpressionObjectPositionWS_0 {class=MaterialExpressionObjectPositionWS, inputs [], properties {}}
  MaterialExpressionObjectPositionWS_1 {class=MaterialExpressionObjectPositionWS, inputs [], properties {originType=CameraRelative}}
  MaterialExpressionObjectPositionWS_2 {class=MaterialExpressionObjectPositionWS, inputs [], properties {}}
  MaterialExpressionObjectPositionWS_4 {class=MaterialExpressionObjectPositionWS, inputs [], properties {originType=CameraRelative}}
  MaterialExpressionObjectPositionWS_6 {class=MaterialExpressionObjectPositionWS, inputs [], properties {}}
  MaterialExpressionObjectPositionWS_7 {class=MaterialExpressionObjectPositionWS, inputs [], properties {}}
  MaterialExpressionObjectPositionWS_8 {class=MaterialExpressionObjectPositionWS, inputs [], properties {}}
  MaterialExpressionReroute_0 {class=MaterialExpressionReroute, inputs [None {from=MaterialExpressionSign_0}]}
  MaterialExpressionReroute_1 {class=MaterialExpressionReroute, inputs [None {from=MaterialExpressionReroute_6}]}
  MaterialExpressionReroute_10 {class=MaterialExpressionReroute, inputs [None {from=MaterialExpressionNamedRerouteDeclaration_5}]}
  MaterialExpressionReroute_11 {class=MaterialExpressionReroute, inputs [None {from=MaterialExpressionStaticSwitchParameter_2}]}
  MaterialExpressionReroute_12 {class=MaterialExpressionReroute, inputs [None {from=MaterialExpressionConstant3Vector_19}]}
  MaterialExpressionReroute_13 {class=MaterialExpressionReroute, inputs [None {from=MaterialExpressionVectorParameter_2, output=RGB}]}
  MaterialExpressionReroute_2 {class=MaterialExpressionReroute, inputs [None {from=MaterialExpressionReroute_7}]}
  MaterialExpressionReroute_3 {class=MaterialExpressionReroute, inputs [None {from=MaterialExpressionMaterialFunctionCall_0, output=New Tangent Space Vertex Normal}]}
  MaterialExpressionReroute_4 {class=MaterialExpressionReroute, inputs [None {from=MaterialExpressionReroute_5}]}
  MaterialExpressionReroute_5 {class=MaterialExpressionReroute, inputs [None {from=MaterialExpressionRotateAboutAxis_0}]}
  MaterialExpressionReroute_6 {class=MaterialExpressionReroute, inputs [None {from=MaterialExpressionTransform_12}]}
  MaterialExpressionReroute_7 {class=MaterialExpressionReroute, inputs [None {from=MaterialExpressionAdd_25}]}
  MaterialExpressionReroute_8 {class=MaterialExpressionReroute, inputs [None {from=""}]}
  MaterialExpressionRotateAboutAxis_0 {
   class=MaterialExpressionRotateAboutAxis
   inputs [
    NormalizedRotationAxis {from=MaterialExpressionConstant3Vector_0, output=""}
    RotationAngle {from=MaterialExpressionAdd_0, output=""}
    PivotPoint {from=MaterialExpressionMaterialFunctionCall_3, output=Result}
    Position {from=MaterialExpressionWorldPosition_1, output=XYZ}
   ]
   properties {}
  }
  MaterialExpressionRotateAboutAxis_2 {
   class=MaterialExpressionRotateAboutAxis
   inputs [
    NormalizedRotationAxis {from=MaterialExpressionNormalize_6, output=""}
    RotationAngle {from=MaterialExpressionScalarParameter_1, output=""}
    PivotPoint {from=MaterialExpressionObjectPositionWS_1, output=""}
    Position {from=MaterialExpressionWorldPosition_2, output=XYZ}
   ]
   properties {}
  }
  MaterialExpressionRotateAboutAxis_5 {
   class=MaterialExpressionRotateAboutAxis
   inputs [
    NormalizedRotationAxis {from=MaterialExpressionNormalize_13, output=""}
    RotationAngle {from=MaterialExpressionScalarParameter_3, output=""}
    PivotPoint {from=MaterialExpressionObjectPositionWS_7, output=""}
    Position {from=MaterialExpressionWorldPosition_5, output=XYZ}
   ]
   properties {}
  }
  MaterialExpressionScalarParameter_0 {class=MaterialExpressionScalarParameter, properties {defaultValue=-1, parameterName=Rotation}}
  MaterialExpressionScalarParameter_1 {class=MaterialExpressionScalarParameter, properties {defaultValue=0.5}}
  MaterialExpressionScalarParameter_17 {class=MaterialExpressionScalarParameter, properties {defaultValue=-0.5, parameterName=LookAtAngleMinClamp}}
  MaterialExpressionScalarParameter_2 {class=MaterialExpressionScalarParameter, properties {defaultValue=-2.3288, parameterName=Param_1}}
  MaterialExpressionScalarParameter_3 {class=MaterialExpressionScalarParameter, properties {defaultValue=0.5, parameterName=Param_2}}
  MaterialExpressionScalarParameter_54 {class=MaterialExpressionScalarParameter, properties {defaultValue=1, parameterName=LookAtAngleMaxClamp}}
  MaterialExpressionScalarParameter_67 {class=MaterialExpressionScalarParameter, properties {defaultValue=0.492, parameterName=LookAtAngleTransitionThreshold}}
  MaterialExpressionScalarParameter_68 {class=MaterialExpressionScalarParameter, properties {defaultValue=0.05, parameterName=LookAtAngleTransitionRange}}
  MaterialExpressionSetMaterialAttributes_0 {
   class=MaterialExpressionSetMaterialAttributes
   inputs [
    MaterialAttributes {from="", output=""}
    Base Color {from=MaterialExpressionTextureSample_0, output=RGB}
    Normal {from=MaterialExpressionStaticSwitchParameter_0, output=""}
    World Position Offset {from=MaterialExpressionStaticSwitchParameter_1, output=""}
   ]
   properties {attributeSetTypes [#3, #1, #2]}
  }
  MaterialExpressionSign_0 {class=MaterialExpressionSign, inputs [None {from=MaterialExpressionDotProduct_1, output=""}], properties {}}
  MaterialExpressionSign_1 {class=MaterialExpressionSign, inputs [None {from=MaterialExpressionVectorParameter_2, output=RGB}], properties {}}
  MaterialExpressionSign_2 {class=MaterialExpressionSign, inputs [None {from=MaterialExpressionComponentMask_3, output=""}], properties {}}
  MaterialExpressionSmoothStep_1 {
   class=MaterialExpressionSmoothStep
   inputs [
    Min {from=MaterialExpressionSubtract_5, output=""}
    Max {from=MaterialExpressionAdd_3, output=""}
    Value {from=MaterialExpressionDotProduct_3, output=""}
   ]
   properties {}
  }
  MaterialExpressionSphereMask_9 {
   class=MaterialExpressionSphereMask
   inputs [
    A {from=MaterialExpressionSmoothStep_1, output=""}
    B {from="", output=""}
    Radius {from="", output=""}
    Hardness {from="", output=""}
   ]
   properties {attenuationRadius=0.5, hardnessPercent=0}
  }
  MaterialExpressionStaticSwitchParameter_0 {
   class=MaterialExpressionStaticSwitchParameter
   inputs [
    True {from=MaterialExpressionReroute_3, output=""}
    False {from=MaterialExpressionReroute_1, output=""}
   ]
   properties {defaultValue=true, parameterName=Alternative Tracking}
  }
  MaterialExpressionStaticSwitchParameter_1 {
   class=MaterialExpressionStaticSwitchParameter
   inputs [
    True {from=MaterialExpressionReroute_4, output=""}
    False {from=MaterialExpressionReroute_2, output=""}
   ]
   properties {defaultValue=true, parameterName=Alternative Tracking}
  }
  MaterialExpressionStaticSwitchParameter_2 {
   class=MaterialExpressionStaticSwitchParameter
   inputs [
    True {from=MaterialExpressionLinearInterpolate_19, output=""}
    False {from=MaterialExpressionNormalize_10, output=""}
   ]
   properties {parameterName=UseTransitionForTopAngle}
  }
  MaterialExpressionStaticSwitchParameter_3 {
   class=MaterialExpressionStaticSwitchParameter
   inputs [
    True {from=MaterialExpressionLinearInterpolate_20, output=""}
    False {from=MaterialExpressionNormalize_11, output=""}
   ]
   properties {parameterName=UseTransitionForTopAngle}
  }
  MaterialExpressionStep_9 {
   class=MaterialExpressionStep
   inputs [
    Y {from=MaterialExpressionScalarParameter_67, output=""}
    X {from=MaterialExpressionDotProduct_3, output=""}
   ]
   properties {}
  }
  MaterialExpressionSubtract_0 {
   class=MaterialExpressionSubtract
   inputs [
    A {from=MaterialExpressionCameraPositionWS_0}
    B {from=MaterialExpressionObjectPositionWS_0}
   ]
  }
  MaterialExpressionSubtract_2 {
   class=MaterialExpressionSubtract
   inputs [
    A {from=MaterialExpressionWorldPosition_4, output=XYZ}
    B {from=MaterialExpressionObjectPositionWS_6}
   ]
  }
  MaterialExpressionSubtract_22 {
   class=MaterialExpressionSubtract
   inputs [
    A {from=MaterialExpressionTransform_21}
    B {from=MaterialExpressionWorldPosition_12, output=XYZ}
   ]
  }
  MaterialExpressionSubtract_3 {
   class=MaterialExpressionSubtract
   inputs [
    A {from=MaterialExpressionWorldPosition_6, output=XYZ}
    B {from=MaterialExpressionObjectPositionWS_8}
   ]
  }
  MaterialExpressionSubtract_4 {
   class=MaterialExpressionSubtract
   inputs [
    A {from=MaterialExpressionNamedRerouteUsage_3}
    B {from=MaterialExpressionTransformPosition_14}
   ]
  }
  MaterialExpressionSubtract_5 {
   class=MaterialExpressionSubtract
   inputs [
    A {from=MaterialExpressionScalarParameter_67}
    B {from=MaterialExpressionScalarParameter_68}
   ]
   properties {constB=0.005}
  }
  MaterialExpressionTextureSample_0 {
   class=MaterialExpressionTextureSample
   inputs [
    UVs {from=""}
    Tex {from=""}
    Apply View MipBias {from=""}
   ]
   properties {texture=/Hera/Props/DarkNexus/DarkNexus_SurveillanceCamera_A/Textures/T_DarkNexus_SurveillanceCamera_A_A_D}
  }
  MaterialExpressionTime_0 {class=MaterialExpressionTime}
  MaterialExpressionTransformPosition_14 {
   class=MaterialExpressionTransformPosition
   inputs [None {from=MaterialExpressionVectorParameter_1, output=RGB}]
   properties {transformType=TRANSFORMPOSSOURCE_World}
  }
  MaterialExpressionTransformPosition_6 {
   class=MaterialExpressionTransformPosition
   inputs [None {from=MaterialExpressionWorldPosition_11, output=XYZ}]
   properties {transformSourceType=TRANSFORMPOSSOURCE_World, transformType=TRANSFORMPOSSOURCE_Instance}
  }
  MaterialExpressionTransform_0 {
   class=MaterialExpressionTransform
   inputs [None {from=MaterialExpressionNormalize_1}]
   properties {transformSourceType=TRANSFORMSOURCE_Local}
  }
  MaterialExpressionTransform_1 {
   class=MaterialExpressionTransform
   inputs [None {from=MaterialExpressionConstant3Vector_5}]
   properties {transformSourceType=TRANSFORMSOURCE_Local}
  }
  MaterialExpressionTransform_12 {
   class=MaterialExpressionTransform
   inputs [None {from=MaterialExpressionMaterialFunctionCall_44, output=Transformed Vector}]
   properties {transformSourceType=TRANSFORMSOURCE_View, transformType=TRANSFORM_Tangent}
  }
  MaterialExpressionTransform_15 {
   class=MaterialExpressionTransform
   inputs [None {from=MaterialExpressionConstant3Vector_1}]
   properties {transformSourceType=TRANSFORMSOURCE_World, transformType=TRANSFORM_Instance}
  }
  MaterialExpressionTransform_18 {
   class=MaterialExpressionTransform
   inputs [None {from=MaterialExpressionMultiply_14}]
   properties {transformSourceType=TRANSFORMSOURCE_World, transformType=TRANSFORM_Instance}
  }
  MaterialExpressionTransform_19 {
   class=MaterialExpressionTransform
   inputs [None {from=MaterialExpressionConstant3Vector_3}]
   properties {transformSourceType=TRANSFORMSOURCE_World, transformType=TRANSFORM_Instance}
  }
  MaterialExpressionTransform_21 {
   class=MaterialExpressionTransform
   inputs [None {from=MaterialExpressionMaterialFunctionCall_26, output=Transformed Vector}]
   properties {transformSourceType=TRANSFORMSOURCE_View}
  }
  MaterialExpressionTransform_3 {
   class=MaterialExpressionTransform
   inputs [None {from=MaterialExpressionMultiply_12}]
   properties {transformSourceType=TRANSFORMSOURCE_Camera}
  }
  MaterialExpressionTransform_4 {
   class=MaterialExpressionTransform
   inputs [None {from=MaterialExpressionConstant3Vector_14}]
   properties {transformSourceType=TRANSFORMSOURCE_Local}
  }
  MaterialExpressionTransform_5 {
   class=MaterialExpressionTransform
   inputs [None {from=MaterialExpressionNormalize_4}]
   properties {transformSourceType=TRANSFORMSOURCE_Local}
  }
  MaterialExpressionTransform_6 {
   class=MaterialExpressionTransform
   inputs [None {from=MaterialExpressionConstant3Vector_20}]
   properties {transformSourceType=TRANSFORMSOURCE_World, transformType=TRANSFORM_Camera}
  }
  MaterialExpressionVectorParameter_1 {class=MaterialExpressionVectorParameter, properties {parameterName=ObjectLocalCenter, defaultValue=<rgba 0 0 1650 0>}}
  MaterialExpressionVectorParameter_2 {class=MaterialExpressionVectorParameter, properties {parameterName=Player Pos, defaultValue=<rgba -6570.1138 -139.6751 2700 1>}}
  MaterialExpressionVectorParameter_23 {
   class=MaterialExpressionVectorParameter
   properties {bUseCustomPrimitiveData=true, parameterName=CurrentPlayerLocation, primitiveDataIndex=8, defaultValue=<rgba 0 0 0 1>}
  }
  MaterialExpressionVectorParameter_4 {class=MaterialExpressionVectorParameter, properties {parameterName=psoition, defaultValue=<rgba -265.0782 0 0 1>}}
  MaterialExpressionVectorParameter_8 {class=MaterialExpressionVectorParameter, properties {parameterName=Param_4, defaultValue=<rgba -85.2139 0 0 1>}}
  MaterialExpressionVertexNormalWS_5 {class=MaterialExpressionVertexNormalWS, inputs [], properties {}}
  MaterialExpressionWorldPosition_0 {class=MaterialExpressionWorldPosition, inputs [], properties {worldPositionShaderOffset=WPT_CameraRelativeNoOffsets}}
  MaterialExpressionWorldPosition_1 {class=MaterialExpressionWorldPosition, inputs [], properties {worldPositionShaderOffset=WPT_ExcludeAllShaderOffsets}}
  MaterialExpressionWorldPosition_11 {class=MaterialExpressionWorldPosition, inputs [], properties {worldPositionShaderOffset=WPT_ExcludeAllShaderOffsets}}
  MaterialExpressionWorldPosition_12 {class=MaterialExpressionWorldPosition, inputs [], properties {}}
  MaterialExpressionWorldPosition_2 {class=MaterialExpressionWorldPosition, inputs [], properties {worldPositionShaderOffset=WPT_CameraRelativeNoOffsets}}
  MaterialExpressionWorldPosition_3 {class=MaterialExpressionWorldPosition, inputs [], properties {worldPositionShaderOffset=WPT_CameraRelative}}
  MaterialExpressionWorldPosition_4 {class=MaterialExpressionWorldPosition, inputs [], properties {}}
  MaterialExpressionWorldPosition_5 {class=MaterialExpressionWorldPosition, inputs [], properties {}}
  MaterialExpressionWorldPosition_6 {class=MaterialExpressionWorldPosition, inputs [], properties {}}
 }
 properties {bAutomaticallySetUsageInEditor=false, bUseMaterialAttributes=true, usedShadingModels=Default Lit}
}
